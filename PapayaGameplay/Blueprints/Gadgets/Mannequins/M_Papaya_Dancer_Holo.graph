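# Translucent unlit hologram material for Papaya dancer mannequins, supporting skeletal meshes and tinted through MF_Papaya_GadgetColor.
# Hue-shifted gradients combine with Fresnel, bent screen-space scanlines, red-green view-normal fringing and an emissive mask.
# A vertical pulse, driven by parameter or Lighting_MPC, brightens and inflates the mesh horizontally.
# A custom-depth comparison culls hidden layers from opacity; low-quality, forward and mobile switches simplify or drop costly effects.
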Material /PapayaGameplay/Blueprints/Gadgets/Mannequins/M_Papaya_Dancer_Holo
graph {
 properties {
  allowTranslucentCustomDepthWrites=true
  bUsedWithMeshDeformer=true
  bUsedWithMorphTargets=true
  bUsedWithSkeletalMesh=true
  blendMode=BLEND_Translucent
  pixelDepthOffsetMode=PDOM_Legacy
  shadingModel=MSM_Unlit
  usedShadingModels=Unlit
 }
 nodes [
  n1 = MaterialFunctionCall!(
   Media Player TexCoord
   Media Color Code Coordinate
   Media Texture Resolution
   Base Color Parameter
   Base Color Mix
   Media Color Code Mix
   Intermission Color Parameter
   Intermission Color Mix
   Desaturation
  ) {materialFunction=/Game/Environments/Papaya/Materials/MaterialFunctions/MF_Papaya_GadgetColor}
  n2 = ScalarParameter() {group=Color Hue And Value, parameterName=Bottom Color Hue Shift}
  n3 = VectorParameter() {parameterName=LightDirection, defaultValue=<rgba 0 0 -1 1>}
  n4 = TextureSampleParameter2D() {
   parameterName=Normal
   samplerType=SAMPLERTYPE_Normal
   texture=/Game/Characters/Player/Female/Medium/Bodies/F_Med_Soldier_01/Skins/BR_39/Textures/F_MED_BR039_01_N
  }
  n5 = Transform(None=@n4.RGB)
  n6 = Power(Base=Clamp(None=DotProduct(A=@n3.RGB, B=@n5))) {constExponent=0.6}
  n7 = Multiply(
   A=MaterialFunctionCall() {materialFunction=/Engine/Functions/Engine_MaterialFunctions02/WorldPositionOffset/ObjectScale}.Scale Z
   B=ScalarParameter() {defaultValue=300, parameterName=GradientSize}
  )
  n8 = Subtract(A=Add(A=ActorPositionWS(), B=Multiply(A=@n7) {constB=0.5}) {constB=50}, B=WorldPosition().XYZ)
  n9 = Clamp(None=Divide(A=ComponentMask(None=@n8) {b=true}, B=@n7))
  n10 = Power(Base=@n9) {constExponent=5}
  n11 = QualitySwitch(
   Default=ShadingPathSwitch(Default=StaticSwitchParameter(True=@n6, False=@n10) {defaultValue=true, parameterName=UseNormal}, Mobile=@n9) {desc=Migrated from FeatureLevelSwitch}
   Low=@n9
  )
  n12 = LinearInterpolate(
   A=Multiply(A=@n1.Result, B=ScalarParameter() {defaultValue=1.2, group=Color Hue And Value, parameterName=Top Color Multiplier})
   B=Multiply(
    A=MaterialFunctionCall(Hue Shift Percentage=@n2, Texture=@n1.Result) {materialFunction=&HueShift}.Result
    B=ScalarParameter() {defaultValue=10, group=Color Hue And Value, parameterName=Bottom Color Multiplier}
   )
   Alpha=@n11
  )
  n13 = Multiply(
   A=MaterialFunctionCall(
    Hue Shift Percentage=ScalarParameter() {defaultValue=0.25, group=Color Hue And Value, parameterName=Bottom Tint Color Hue Shift}
    Texture=@n1.Result
   ) {materialFunction=&HueShift}.Result
   B=ScalarParameter() {defaultValue=3, group=Color Hue And Value, parameterName=Bottom Tint Color Multiplier}
  )
  n14 = MaterialFunctionCall() {materialFunction=&ObjectScale}
  n15 = MaterialFunctionCall(Distance=Multiply(A=ScalarParameter() {defaultValue=1500, parameterName=FadeDistance}, B=@n14.Scale Z)) {materialFunction=&Near_Camera_Fade}
  n16 = LinearInterpolate(Alpha=@n15.Result) {constA=3.2}
  n17 = Constant() {r=3}
  n18 = Fresnel(
   ExponentIn=ShadingPathSwitch(Default=QualitySwitch(Default=@n16, Low=@n17), Forward=@n17)
   BaseReflectFractionIn=LinearInterpolate(
    A=ScalarParameter() {defaultValue=0.1, parameterName=BaseRefractMin}
    B=ScalarParameter() {defaultValue=0.3, parameterName=BaseRefract}
    Alpha=@n11
   ) {constA=0.1}
  ) {exponent=3}
  n19 = VertexNormalWS()
  n20 = ComponentMask(None=@n19) {b=true}
  n21 = ComponentMask(None=CameraVectorWS()) {b=true}
  n22 = Multiply(
   A=Fresnel() {baseReflectFraction=0, exponent=3}
   B=Multiply(A=DotProduct(A=@n20, B=@n21), B=ScalarParameter() {defaultValue=0.5, parameterName=LineBend})
  ) {constB=0.2}
  n23 = Constant()
  n24 = ScreenPosition()
  n25 = Multiply(
   A=Add(A=QualitySwitch(Default=@n22, Low=@n23, Medium=@n23), B=ComponentMask(None=@n24.ViewportUV) {g=true})
   B=ScalarParameter() {defaultValue=200, parameterName=ScanlineTile}
  )
  n26 = Power(Base=ConstantBiasScale(None=Sine(None=@n25)))
  n27 = Constant3Vector() {constant=<rgba 1 0 0.1 0>}
  n28 = Constant3Vector() {constant=<rgba 0 1 0.1 0>}
  n29 = ConstantBiasScale(None=ComponentMask(None=Transform(None=@n19) {transformSourceType=TRANSFORMSOURCE_World, transformType=TRANSFORM_View}) {r=true})
  n30 = Add(A=LinearInterpolate(A=@n27, B=@n28, Alpha=@n29)) {constB=0.4}
  n31 = Multiply(A=CollectionParameter() {collection=/Game/Athena/Environments/Festivus/MotorizedLights/Lighting_MPC, parameterName=HoloGlow}) {constB=7}
  n32 = Add(A=LinearInterpolate(B=@n30, Alpha=@n26) {constA=1}, B=Multiply(A=@n30, B=@n31))
  n33 = Add(A=Constant() {r=1}, B=@n31)
  n34 = Constant() {r=1}
  n35 = ShadingPathSwitch(Default=QualitySwitch(Default=LinearInterpolate(A=@n28, B=@n27, Alpha=@n29), Low=@n34), Forward=@n34)
  n36 = ScalarParameter() {group=Color Hue And Value, parameterName=Pulse Color Hue Shift}
  n37 = WorldPosition()
  n38 = ComponentMask(None=ActorPositionWS()) {b=true}
  n39 = Clamp(None=Max(
    A=ScalarParameter() {parameterName=PulsePosition}
    B=CollectionParameter() {collection=&Lighting_MPC, parameterName=HoloPulseGameTime}
   ))
  n40 = Multiply(A=LinearInterpolate(Alpha=@n39) {constA=1, constB=0.1}, B=ScalarParameter() {defaultValue=35, parameterName=PulseLineRadius})
  n41 = MaterialFunctionCall() {materialFunction=&ObjectScale}
  n42 = Multiply(
   A=LinearInterpolate(A=Multiply(A=@n40) {constB=-0.5}, B=ScalarParameter() {defaultValue=200, parameterName=PulseLineTop}, Alpha=@n39) {constB=0}
   B=@n41.Scale Z
  )
  n43 = Add(A=Multiply(A=@n41.Scale Z) {constB=0.5}) {constB=0.5}
  n44 = Power(Base=SphereMask(A=ComponentMask(None=@n37.XYZ) {b=true}, B=Add(A=@n38, B=@n42), Radius=Multiply(A=@n43, B=@n40)) {hardnessPercent=0})
  n45 = Multiply(
   A=Multiply(
    A=MaterialFunctionCall(Hue Shift Percentage=@n36, Texture=@n1.Result) {materialFunction=&HueShift}.Result
    B=ScalarParameter() {defaultValue=250, group=Color Hue And Value, parameterName=Pulse Color Multiplier}
   )
   B=@n44
  )
  n46 = Add(
   A=ShadingPathSwitch(Default=QualitySwitch(Default=@n32, Low=@n33), Forward=@n33)
   B=ShadingPathSwitch(Default=Multiply(A=@n35, B=@n45), Mobile=Constant()) {desc=Migrated from FeatureLevelSwitch}
  )
  n47 = MaterialFunctionCall(
   Hue Shift Percentage=ScalarParameter() {defaultValue=0.2, group=Color Hue And Value, parameterName=Emissive Mask Color Hue Shift}
   Texture=@n1.Result
  ) {materialFunction=&HueShift}
  n48 = ScalarParameter() {defaultValue=10, group=Color Hue And Value, parameterName=Emissive Mask Color Multiplier}
  n49 = StaticComponentMaskParameter(None=TextureSampleParameter2D() {parameterName=Emissive, texture=/Game/Athena/Environments/Festivus/Materials/T_DanceBear_E}.RGB) {defaultG=true, parameterName=EmissiveMask}
  Add(
   A=Multiply(A=LinearInterpolate(A=@n12, B=@n13, Alpha=@n9), B=Multiply(A=@n18, B=@n46))
   B=Multiply(A=Multiply(A=@n47.Result, B=@n48), B=@n49)
  )
  n50 = Multiply(A=@n49) {constB=0.4}
  n51 = Add(
   A=Add(A=Add(A=@n18, B=@n50), B=ScalarParameter() {defaultValue=0.7, parameterName=ScanlineAdd})
   B=LinearInterpolate(A=@n26, Alpha=@n15.Result) {constA=0.4, constB=0.3}
  ) {constB=0.2}
  n52 = ScalarParameter() {defaultValue=-1, parameterName=BiasSorting_Far}
  n53 = Constant()
  n54 = If(
   A=Add(A=PixelDepth().R, B=Multiply(A=@n14.Scale Z, B=@n52))
   B=ComponentMask(None=SceneTexture() {sceneTextureId=PPI_CustomDepth}.Color) {r=true}
   "A > B"=@n53
   "A == B"=@n53
   "A < B"=Constant() {r=1}
  )
  n55 = Constant() {r=1}
  Multiply(
   A=Multiply(A=Clamp(None=@n51) {maxDefault=2}, B=ShadingPathSwitch(Default=@n54, Mobile=@n55) {desc=Migrated from FeatureLevelSwitch})
   B=ScalarParameter() {defaultValue=0.6, parameterName=Overall Opacity}
  )
  n56 = Multiply(
   A=ScalarParameter() {defaultValue=6, parameterName=PulseWPOMult}
   B=Multiply(A=VertexNormalWS(), B=Constant3Vector() {constant=<rgba 1 1 0 0>})
  )
  Reroute(None=ShadingPathSwitch(Default=Multiply(A=@n44, B=@n56), Mobile=Constant()) {desc=Migrated from FeatureLevelSwitch})
  Time() {bOverride_Period=true, period=1}
 ]
}
